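# Sound cue for the party hover sound in the front-end UI, routed through the Fort_UI sound class.
# A single wave player plays the Fort_UI_Party_Hover wave with no randomization or modulation.
# A concurrency override allows only one instance at a time and stops the oldest, so repeated hovers cut off the previous sound instead of stacking.
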SoundCue /Game/Sounds/Fort_UI_Sounds/FrontEnd/Fort_UI_Party_Hover_Cue
root=WavePlayer_0
nodes [{id=WavePlayer_0, properties {soundWaveAssetPtr=&/Fort_UI_Party_Hover}}]
properties {
 bOverrideConcurrency=true
 maxDistance=2097152
 soundClassObject=&Fort_UI
 concurrencyOverrides {maxCount=1, resolutionRule=StopOldest}
}
