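# Frontend party suggestion button on FortPartySuggestionButton, showing a suggested player's name and status with an invite prompt.
# It starts collapsed; new or changed suggestions refresh display info and play the intro, while dismissal plays the outro.
# Hover scales the button, reveals prompt and sparkles, focuses the outline and plays sound, reversing on unhover.
# Accepting plays the invite-sent sound, plus a pulse animation unless the player is already invited.
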WidgetBlueprint /Game/Athena/UI/Frontend/Parties/WBP_PartySuggestionButton
animations [
 Anim_Intro {
  binding_count=2
  length_seconds=0.35
  object=&self:Anim_Intro
  sequence {
   bindings [
    VB_Content {
     id=#1
     possessed_class=VerticalBox
     tracks [{
       type=MovieSceneFloatTrack
       display_name=Render Opacity
       sections [{
         type=MovieSceneFloatSection
         end_frame=35
         start_frame=0
         channels [None {type=MovieSceneScriptingFloatChannel, keys [[0,0],[20,0],[30,1],[34.9983,1]]}]
        }]
      }]
    }
    GP_Master {
     id=#2
     possessed_class=Overlay
     tracks [
      {
       type=MovieScene2DTransformTrack
       sections [{
         type=MovieScene2DTransformSection
         end_frame=35
         start_frame=0
         channels [
          Translation.X {type=MovieSceneScriptingFloatChannel, const [0,2]}
          Translation.Y {type=MovieSceneScriptingFloatChannel, const [0,2]}
          Angle {type=MovieSceneScriptingFloatChannel}
          Scale.X {type=MovieSceneScriptingFloatChannel, keys [[0,0],[15,1.1871],[25,0.9198],[30,1.0345],[34.9983,1]]}
          Scale.Y {type=MovieSceneScriptingFloatChannel, keys [[0,0],[15,1.1961],[25,0.9275],[30,1.0575],[34.9983,1]]}
          Shear.X {type=MovieSceneScriptingFloatChannel}
          Shear.Y {type=MovieSceneScriptingFloatChannel}
         ]
        }]
      }
      {
       type=MovieSceneFloatTrack
       display_name=Render Opacity
       sections [{type=MovieSceneFloatSection, end_frame=35, start_frame=0, channels [None {type=MovieSceneScriptingFloatChannel, keys [[0,0],[35,1]]}]}]
      }
     ]
    }
   ]
   display_rate {numerator=100}
  }
 }
 Anim_Outro {
  binding_count=1
  length_seconds=0.3
  object=&self:Anim_Outro
  sequence {bindings [GP_Master {
     id=#3
     possessed_class=Overlay
     tracks [
      {
       type=MovieScene2DTransformTrack
       sections [{
         type=MovieScene2DTransformSection
         end_frame=6
         start_frame=0
         channels [
          Translation.X {type=MovieSceneScriptingFloatChannel, const [0,2]}
          Translation.Y {type=MovieSceneScriptingFloatChannel, keys [[0,0],[6,-50]]}
          Angle {type=MovieSceneScriptingFloatChannel}
          Scale.X {type=MovieSceneScriptingFloatChannel}
          Scale.Y {type=MovieSceneScriptingFloatChannel}
          Shear.X {type=MovieSceneScriptingFloatChannel}
          Shear.Y {type=MovieSceneScriptingFloatChannel}
         ]
        }]
      }
      {
       type=MovieSceneFloatTrack
       display_name=Render Opacity
       sections [{type=MovieSceneFloatSection, end_frame=6, start_frame=0, channels [None {type=MovieSceneScriptingFloatChannel, keys [[0,1],[6,0]]}]}]
      }
     ]
    }]}
 }
 MouseHovered {
  binding_count=3
  length_seconds=0.3
  object=&self:MouseHovered
  sequence {
   bindings [
    GP_Master {
     id=#4
     possessed_class=Overlay
     tracks [{
       type=MovieScene2DTransformTrack
       sections [{
         type=MovieScene2DTransformSection
         channels [
          Translation.X {type=MovieSceneScriptingFloatChannel}
          Translation.Y {type=MovieSceneScriptingFloatChannel}
          Angle {type=MovieSceneScriptingFloatChannel}
          Scale.X {type=MovieSceneScriptingFloatChannel, keys [[0,1],[12,1.1],[18,1.1]]}
          Scale.Y {type=MovieSceneScriptingFloatChannel, keys [[0,1],[12,1.1],[18,1.1]]}
          Shear.X {type=MovieSceneScriptingFloatChannel}
          Shear.Y {type=MovieSceneScriptingFloatChannel}
         ]
        }]
      }]
    }
    SparklyTraingles {
     id=#5
     possessed_class=Image
     tracks [{
       type=MovieSceneFloatTrack
       display_name=Render Opacity
       sections [{type=MovieSceneFloatSection, channels [None {type=MovieSceneScriptingFloatChannel, keys [[0,0],[9,1],[18,1]]}]}]
      }]
    }
    HB_ActionPrompt {
     id=#6
     possessed_class=HorizontalBox
     tracks [
      {
       type=MovieSceneFloatTrack
       display_name=Render Opacity
       sections [{type=MovieSceneFloatSection, channels [None {type=MovieSceneScriptingFloatChannel, keys [[0,0],[18,1]]}]}]
      }
      {
       type=MovieScene2DTransformTrack
       sections [{
         type=MovieScene2DTransformSection
         channels [
          Translation.X {type=MovieSceneScriptingFloatChannel}
          Translation.Y {type=MovieSceneScriptingFloatChannel, keys [[0,30],[18,-52]]}
          Angle {type=MovieSceneScriptingFloatChannel}
          Scale.X {type=MovieSceneScriptingFloatChannel, keys [[0,0.5],[18,1]]}
          Scale.Y {type=MovieSceneScriptingFloatChannel, keys [[0,0.8],[18,1]]}
          Shear.X {type=MovieSceneScriptingFloatChannel}
          Shear.Y {type=MovieSceneScriptingFloatChannel}
         ]
        }]
      }
     ]
    }
   ]
   display_rate {numerator=60}
  }
 }
 Anim_RequestAccepted {
  binding_count=1
  length_seconds=0.2
  object=&self:Anim_RequestAccepted
  sequence {bindings [GridPanel_0 {
     id=#7
     possessed_class=GridPanel
     tracks [{
       type=MovieScene2DTransformTrack
       sections [{
         type=MovieScene2DTransformSection
         end_frame=4
         start_frame=0
         channels [
          Translation.X {type=MovieSceneScriptingFloatChannel, const [0,1]}
          Translation.Y {type=MovieSceneScriptingFloatChannel, const [0,2]}
          Angle {type=MovieSceneScriptingFloatChannel, const [0,1]}
          Scale.X {type=MovieSceneScriptingFloatChannel, keys [[0,1],[2,1.05],[4,1]]}
          Scale.Y {type=MovieSceneScriptingFloatChannel, const [1,2]}
          Shear.X {type=MovieSceneScriptingFloatChannel, const [0,1]}
          Shear.Y {type=MovieSceneScriptingFloatChannel, const [0,1]}
         ]
        }]
      }]
    }]}
 }
]
graphs [
 EventGraph {nodes [
   Event OnSuggestionAccepted() {is=title K2Node_Event}
   > PlayRequestAccepted()
   Event On Hovered()
   > SetTransitionFocus!(execute, self, IsFocused:bool=true) -> (then)
   Event Construct()
   > SetRenderOpacity(
    self=[
     Get HB_ActionPrompt() -> (HB_ActionPrompt:Horizontal Box Object Reference)
     Get SparklyTraingles() -> (SparklyTraingles:Image Object Reference)
    ]
    InOpacity=0
   )
   > SetVisibility(InVisibility=Collapsed)
   Event On Unhovered()
   > SetTransitionFocus(IsFocused=false)
   n1 = Event OnDisplayInfoUpdated!() -> (OutputDelegate, then, bIsInvited:bool, bIsPlaying:bool) {is=title K2Node_Event}
   > Set bIsInvited(bIsInvited:bool=@n1.bIsInvited) -> (Output_Get:bool)
   Event OnDismissSuggestion() {is=title K2Node_Event}
   > n2 = Branch(Condition=IsAnimationPlaying(InAnimation=Get Anim_Outro() -> (Anim_Outro:Widget Animation Object Reference)))
   StopAnimation(execute=@n2.else, InAnimation=Get Anim_Intro() -> (Anim_Intro:Widget Animation Object Reference))
   > PlayAnimation(InAnimation=Get Anim_Outro())
   n3 = Event OnSuggestionChanged() {is=title K2Node_Event}
   UpdateDisplayInfo(execute=[@n3.then, @n4.then], self:Fort Party Suggestion Button Object Reference)
   > PlayIntro()
   n4 = Event OnNewSuggestionAdded() {is=title K2Node_Event}
  ]}
 PlayIntro {nodes [
   PlayIntro() {is=title K2Node_FunctionEntry}
   > StopAnimation(InAnimation=Get Anim_Outro())
   > SetVisibility(InVisibility=Visible)
   > PlayAnimation(InAnimation=Get Anim_Intro())
  ]}
 PlayRequestAccepted {nodes [
   PlayRequestAccepted() {is=title K2Node_FunctionEntry}
   > PlaySound2D(Sound=/Game/Sounds/Fort_UI_Sounds/FrontEnd/Fort_UI_Party_InviteSent_Cue)
   > n1 = Branch(Condition=Get bIsInvited() -> (bIsInvited:bool))
   PlayAnimation(execute=@n1.else, InAnimation=Get Anim_RequestAccepted() -> (Anim_RequestAccepted:Widget Animation Object Reference))
  ]}
 SetTransitionFocus {nodes [
   n1 = SetTransitionFocus!() -> (then, IsFocused:bool=~) {is=title K2Node_FunctionEntry}
   > n2 = Branch(Condition=@n1)
   n3 = Get MouseHovered() -> (MouseHovered:Widget Animation Object Reference)
   > PlayAnimationForward(InAnimation=@n3)
   n4 = Get Block_Outline() -> (Block_Outline:WBP UIKit Block Outline Object Reference)
   > RequestTransition!(execute, self:UIKit Block Object Reference=@n4, State:EUIKitBlockVisualState Enum=Focused) -> (then)
   > Is Valid(InputObject=@n4)
   > PlaySound2D(Sound=/Game/Sounds/Fort_UI_Sounds/FrontEnd/Fort_UI_Party_Hover_Cue)
   PlayAnimationReverse(execute=@n2.else, InAnimation=@n3, PlaybackSpeed=1.3)
   > RequestTransition(self=@n4, State=Unfocused)
   > Is Valid(InputObject=@n4)
   > PlaySound2D(Sound=/Game/Sounds/Fort_UI_Sounds/FrontEnd/Fort_UI_Party_UnHover_Cue)
  ]}
]
variables [bIsInvited {type=boolean}]
widgets {
 info {inheritedWidgetCount=4, parentClass=/Script/FortniteUI.FortPartySuggestionButton, rootWidgetClass=&GridPanel, widgetCount=17}
 widgets [
  GP_Master {widgetClassPath=&GridPanel, widget=GP_Master, properties {columnFill [1], rowFill [1]}}
  FortVisualAttachment_0 {
   widgetClassPath=&FortVisualAttachment
   parent=GP_Master
   widget=FortVisualAttachment_0
   properties {contentAnchor=<0.5 0>}
   slot {class=GridSlot, properties {horizontalAlignment=HAlign_Center, verticalAlignment=VAlign_Top}}
  }
  SparklyTraingles {
   widgetClassPath=&Image
   bIsVariable=true
   parent=FortVisualAttachment_0
   widget=SparklyTraingles
   properties {
    visibility=SelfHitTestInvisible
    brush {
     bIsDynamicallyLoaded=false
     drawAs=Image
     imageType=NoImage
     mirroring=NoMirror
     resourceName=None
     resourceObject=/Game/UI/Foundation/Onboarding/Materials/M_FX_TriangleScroll-PartySuggestion
     tiling=NoTile
     imageSize=<104.0479 105.5286>
     margin=<ltrb 0 0 0 0>
     outlineSettings {
      bUseBrushTransparency=false
      roundingType=HalfHeightRadius
      width=0
      color {colorUseRule=UseColor_Specified, specifiedColor=<rgba 0 0 0 0>}
      cornerRadii=<0 0 0 0>
     }
     tintColor {colorUseRule=UseColor_Specified, specifiedColor=<rgba 1 1 1 1>}
     uVRegion {bIsValid=false, max=<0 0>, min=<0 0>}
    }
    colorAndOpacity=<rgba 1 1 1 0.8>
    renderTransform {angle=-90, scale=<1.9 1>, shear=<0 0>, translation=<0 -23>}
    renderTransformPivot=<0.25 0.25>
   }
   slot {class=SizeBoxSlot}
  }
  FortVisualAttachment_1 {
   widgetClassPath=&FortVisualAttachment
   parent=GP_Master
   widget=FortVisualAttachment_1
   properties {contentAnchor=<0.5 0>}
   slot {class=GridSlot, properties {horizontalAlignment=HAlign_Center, layer=10, verticalAlignment=VAlign_Top}}
  }
  HB_ActionPrompt {
   widgetClassPath=&HorizontalBox
   bIsVariable=true
   parent=FortVisualAttachment_1
   widget=HB_ActionPrompt
   properties {renderTransform {angle=0, scale=<1 1>, shear=<0 0>, translation=<0 -52>}}
   slot {class=SizeBoxSlot}
  }
  Text_ActionText {
   widgetClassPath=&CommonTextBlock
   bInherited=true
   bIsVariable=true
   parent=HB_ActionPrompt
   widget=Text_ActionText
   properties {
    text=Invite to Party
    font {
     bForceMonospaced=false
     bMaterialIsStencil=false
     fontMaterial=null
     fontObject=&BurbankSmall
     letterSpacing=0
     monospacedWidth=1
     size=21
     skewAmount=0
     typefaceFontName=Black
     outlineSettings {
      bApplyOutlineToDropShadows=false
      bMiteredCorners=false
      bSeparateFillAlpha=false
      outlineMaterial=null
      outlineSize=0
      outlineColor=<rgba 0 0 0 1>
     }
    }
    shadowColorAndOpacity=<rgba 0.0052 0.0052 0.0056 0.3>
    shadowOffset=<3 3>
   }
   slot {class=HorizontalBoxSlot}
  }
  GridPanel_0 {
   widgetClassPath=&GridPanel
   parent=GP_Master
   widget=GridPanel_0
   properties {visibility=HitTestInvisible, columnFill [1], rowFill [1]}
   slot {class=GridSlot, properties {layer=20}}
  }
  "UIKit_Backplate\\" {
   widgetClassPath=&WBP_UIKit_Backplate
   parent=GridPanel_0
   widget="UIKit_Backplate\\"
   properties {brightness=Dark, cornerRadius=Interactive Component}
   slot {class=GridSlot}
  }
  BorderOfDarkenedText {
   widgetClassPath=&Border
   bIsVariable=true
   parent=GridPanel_0
   widget=BorderOfDarkenedText
   properties {
    background {
     bIsDynamicallyLoaded=false
     drawAs=NoDrawType
     imageType=NoImage
     mirroring=NoMirror
     resourceName=None
     resourceObject=null
     tiling=NoTile
     imageSize=<32 32>
     margin=<ltrb 0 0 0 0>
     outlineSettings {
      bUseBrushTransparency=false
      roundingType=HalfHeightRadius
      width=0
      color {colorUseRule=UseColor_Specified, specifiedColor=<rgba 0 0 0 0>}
      cornerRadii=<0 0 0 0>
     }
     tintColor {colorUseRule=UseColor_Specified, specifiedColor=<rgba 1 1 1 1>}
     uVRegion {bIsValid=false, max=<0 0>, min=<0 0>}
    }
    padding=<ltrb 0 0 0 0>
   }
   slot {class=GridSlot, properties {layer=1, padding=<ltrb 24 16 24 16>}}
  }
  VB_Content {
   widgetClassPath=&VerticalBox
   parent=BorderOfDarkenedText
   widget=VB_Content
   slot {class=BorderSlot, properties {padding=<ltrb 0 0 0 0>}}
  }
  Text_SocialName {
   widgetClassPath=/Script/FortniteUI.FortSocialNameTextBlock
   bInherited=true
   bIsVariable=true
   parent=VB_Content
   widget=Text_SocialName
   properties {
    bOverrideDefaultStyle=true
    bTintInlineIcon=true
    bWrapToNextLine=true
    text="Nickname\n<PC color=\"E7ECECFF\" size=\"12\"></><text color=\"E7ECECFF\" size=\"12\">Display Name</>"
    defaultTextStyleOverride {
     overflowPolicy=Clip
     transformPolicy=None
     colorAndOpacity {colorUseRule=UseColor_Specified, specifiedColor=<rgba 1 1 1 1>}
     font {
      bForceMonospaced=false
      bMaterialIsStencil=false
      fontMaterial=null
      fontObject=&HeadingNow
      letterSpacing=10
      monospacedWidth=1
      size=15.75
      skewAmount=0
      typefaceFontName=75 Medium
      outlineSettings {
       bApplyOutlineToDropShadows=false
       bMiteredCorners=false
       bSeparateFillAlpha=false
       outlineMaterial=null
       outlineSize=0
       outlineColor=<rgba 0 0 0 1>
      }
     }
     highlightColor {colorUseRule=UseColor_Specified, specifiedColor=<rgba 0 0 0 1>}
     highlightShape {
      bIsDynamicallyLoaded=false
      drawAs=Image
      imageType=NoImage
      mirroring=NoMirror
      resourceName=None
      resourceObject=null
      tiling=NoTile
      imageSize=<32 32>
      margin=<ltrb 0 0 0 0>
      outlineSettings {
       bUseBrushTransparency=false
       roundingType=HalfHeightRadius
       width=0
       color {colorUseRule=UseColor_Specified, specifiedColor=<rgba 0 0 0 0>}
       cornerRadii=<0 0 0 0>
      }
      tintColor {colorUseRule=UseColor_Specified, specifiedColor=<rgba 1 1 1 1>}
      uVRegion {bIsValid=false, max=<0 0>, min=<0 0>}
     }
     selectedBackgroundColor {colorUseRule=UseColor_Foreground, specifiedColor=<rgba 1 0 1 1>}
     shadowColorAndOpacity=<rgba 0 0 0 1>
     shadowOffset=<0 0>
     strikeBrush {
      bIsDynamicallyLoaded=false
      drawAs=Image
      imageType=NoImage
      mirroring=NoMirror
      resourceName=None
      resourceObject=null
      tiling=NoTile
      imageSize=<32 32>
      margin=<ltrb 0 0 0 0>
      outlineSettings {
       bUseBrushTransparency=false
       roundingType=HalfHeightRadius
       width=0
       color {colorUseRule=UseColor_Specified, specifiedColor=<rgba 0 0 0 0>}
       cornerRadii=<0 0 0 0>
      }
      tintColor {colorUseRule=UseColor_Specified, specifiedColor=<rgba 1 1 1 1>}
      uVRegion {bIsValid=false, max=<0 0>, min=<0 0>}
     }
     underlineBrush {
      bIsDynamicallyLoaded=false
      drawAs=Image
      imageType=NoImage
      mirroring=NoMirror
      resourceName=None
      resourceObject=null
      tiling=NoTile
      imageSize=<32 32>
      margin=<ltrb 0 0 0 0>
      outlineSettings {
       bUseBrushTransparency=false
       roundingType=HalfHeightRadius
       width=0
       color {colorUseRule=UseColor_Specified, specifiedColor=<rgba 0 0 0 0>}
       cornerRadii=<0 0 0 0>
      }
      tintColor {colorUseRule=UseColor_Specified, specifiedColor=<rgba 1 1 1 1>}
      uVRegion {bIsValid=false, max=<0 0>, min=<0 0>}
     }
    }
    secondNameColor=<rgba 0.9059 0.9255 0.9255 1>
   }
   slot {class=VerticalBoxSlot}
  }
  HorizontalBox_3 {
   widgetClassPath=&HorizontalBox
   parent=VB_Content
   widget=HorizontalBox_3
   slot {class=VerticalBoxSlot, properties {horizontalAlignment=HAlign_Left, verticalAlignment=VAlign_Center}}
  }
  Image_PartySize {
   widgetClassPath=&Image
   bIsVariable=true
   parent=HorizontalBox_3
   widget=Image_PartySize
   properties {
    visibility=Collapsed
    brush {
     bIsDynamicallyLoaded=false
     drawAs=Image
     imageType=NoImage
     mirroring=NoMirror
     resourceName=None
     resourceObject=/FortUILibrary/Textures/MaterialInstances/MI_UIKit_Icon_CurrentPlayerCount
     tiling=NoTile
     imageSize=<32 32>
     margin=<ltrb 0 0 0 0>
     outlineSettings {
      bUseBrushTransparency=false
      roundingType=HalfHeightRadius
      width=0
      color {colorUseRule=UseColor_Specified, specifiedColor=<rgba 0 0 0 0>}
      cornerRadii=<0 0 0 0>
     }
     tintColor {colorUseRule=UseColor_Specified, specifiedColor=<rgba 1 1 1 1>}
     uVRegion {bIsValid=false, max=<0 0>, min=<0 0>}
    }
   }
   slot {class=HorizontalBoxSlot, properties {horizontalAlignment=HAlign_Left, verticalAlignment=VAlign_Center, padding=<ltrb 0 0 4 0>}}
  }
  SizeBox_79 {
   widgetClassPath=&SizeBox
   parent=HorizontalBox_3
   widget=SizeBox_79
   properties {bOverride_MaxDesiredWidth=true, maxDesiredWidth=300}
   slot {class=HorizontalBoxSlot}
  }
  Text_Description {
   widgetClassPath=&CommonTextBlock
   bInherited=true
   bIsVariable=true
   parent=SizeBox_79
   widget=Text_Description
   properties {
    scrollStyle=&TextScrollStyle-Base_C
    text=is Online
    colorAndOpacity {colorUseRule=UseColor_Specified, specifiedColor=<rgba 0.0369 0.9823 0.227 1>}
    font {
     bForceMonospaced=false
     bMaterialIsStencil=false
     fontMaterial=null
     fontObject=&BurbankSmall
     letterSpacing=0
     monospacedWidth=1
     size=15.75
     skewAmount=0
     typefaceFontName=Bold
     outlineSettings {
      bApplyOutlineToDropShadows=false
      bMiteredCorners=false
      bSeparateFillAlpha=false
      outlineMaterial=null
      outlineSize=0
      outlineColor=<rgba 0 0 0 1>
     }
    }
    shadowOffset=<0 0>
   }
   slot {class=SizeBoxSlot, properties {verticalAlignment=VAlign_Center}}
  }
  Block_Outline {
   widgetClassPath=/FortUILibrary/UIKit/Buttons/Blocks/WBP_UIKit_Block_Outline
   bIsVariable=true
   parent=GridPanel_0
   widget=Block_Outline
   properties {bottomLeftCornerRadius=28, bottomRightCornerRadius=28, radiusInPixels=true, topLeftCornerRadius=28, topRightCornerRadius=28}
   slot {class=GridSlot, properties {layer=3}}
  }
  InputActionWidget {widgetClassPath=&CommonActionWidget, bInherited=true, widget=null}
 ]
}
